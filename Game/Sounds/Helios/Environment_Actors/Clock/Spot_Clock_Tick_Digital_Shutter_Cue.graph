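# Spot ambient Sound Cue for a digital shutter clock tick, classed as an Athena ambient prop sound.
# A Random node picks evenly between two tick waves and feeds a Modulator that holds volume steady, while the cue plays quiet and pitched up.
# Overridden attenuation uses logarithmic falloff with listener focus, and overridden concurrency tightly caps simultaneous ticks.
# It also sends to an exterior ambient source bus and the default reverb submix.
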SoundCue /Game/Sounds/Helios/Environment_Actors/Clock/Spot_Clock_Tick_Digital_Shutter_Cue
root=Modulator_0
nodes [
 {id=Modulator_0, children [Random_0], properties {volumeMax=1, volumeMin=1}}
 {id=Random_0, children [WavePlayer_0,WavePlayer_1], properties {weights [1,1]}}
 {id=WavePlayer_0, properties {soundWaveAssetPtr=&/Spot_Clock_Tick_Digital_Shutter_01}}
 {id=WavePlayer_1, properties {soundWaveAssetPtr=&/Spot_Clock_Tick_Digital_Shutter_02}}
]
properties {
 attenuationSettings=/Game/Sounds/Attenuation/Fort_BigClock_Att
 bOverrideAttenuation=true
 bOverrideConcurrency=true
 maxDistance=750
 pitchMultiplier=2
 priority=0
 soundClassObject=&Fort_Ambient_Props_Athena
 volumeMultiplier=0.033
 attenuationOverrides {
  bEnableListenerFocus=true
  bEnablePriorityAttenuation=true
  distanceAlgorithm=Logarithmic
  falloffDistance=750
  focusAzimuth=20
  nonFocusAzimuth=40
  nonFocusVolumeAttenuation=0
  priorityAttenuationMax=0
  reverbDistanceMax=1000
  stereoSpread=50
  attenuationShapeExtents {x=0}
 }
 busSends [{audioBus=null, minSendLevel=2, sendLevel=2, soundSourceBus=&AmbientAudio_Exterior_SourceBus}]
 concurrencyOverrides {maxCount=2, voiceStealReleaseTime=0.5}
 soundSubmixSends [{minSendLevel=1, soundSubmix=&Fortnite_Reverb_Default}]
}
